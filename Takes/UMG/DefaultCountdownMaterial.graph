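# UI translucent material for a countdown indicator, driven by the CountdownTime scalar parameter.
# The fractional time, scaled to a full turn, is compared with an arctangent angle from centred UVs, using Ceil and Saturate for a hard mask.
# As written, both arctangent inputs take the R component, rather than separate X and Y.
# The mask selects pale gold over a background that blends from gold to dark as the fraction rises.
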Material /Takes/UMG/DefaultCountdownMaterial
graph {
 properties {blendMode=BLEND_Translucent, materialDomain=MD_UI, pixelDepthOffsetMode=PDOM_Legacy, usedShadingModels=Default Lit}
 nodes [
  Constant() {r=0.5}
  n1 = Constant3Vector() {constant=<rgba 0.6292 0.5645 0.3146 0>}
  n2 = Fmod(A=ScalarParameter() {defaultValue=0.89, parameterName=CountdownTime}, B=Constant() {r=1})
  n3 = Constant() {r=2}
  n4 = MaterialFunctionCall(Float2=Subtract(A=TextureCoordinate(), B=Constant2Vector() {g=0.5, r=0.5})) {materialFunction=&BreakOutFloat2Components}
  n5 = Constant() {r=1}
  n6 = Subtract(
   A=Multiply(A=@n2, B=MaterialFunctionCall(Multiplier=@n3) {materialFunction=&Pi}.Result)
   B=Add(
    A=Arctangent2!(Y=@n4.R, X=@n4.R)
    B=MaterialFunctionCall(Multiplier=@n5) {materialFunction=&Pi}.Result
   )
  )
  LinearInterpolate(
   A=LinearInterpolate(A=Constant3Vector() {constant=<rgba 0.075 0.0641 0.0219 0>}, B=@n1, Alpha=OneMinus(None=@n2))
   B=@n1
   Alpha=Saturate(None=Ceil(None=@n6))
  )
 ]
}
