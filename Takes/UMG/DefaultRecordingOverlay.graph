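# Take Recorder overlay widget that draws a film-leader style countdown with crosshairs, circles and a large number before recording.
# PreConstruct creates a dynamic instance of the countdown material for the full-screen border brush.
# Tick feeds the recorder's remaining seconds into its CountdownTime parameter to drive the material.
# Helper functions return those seconds rounded up as text, and visibility only while the recorder state is CountingDown.
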WidgetBlueprint /Takes/UMG/DefaultRecordingOverlay
graphs [
 EventGraph {nodes [
   Event PreConstruct()
   > n1 = CreateDynamicMaterialInstance!(execute, Parent=&/DefaultCountdownMaterial, OptionalName=None, CreationFlags=None) -> (then, ReturnValue)
   > n2 = Set MID_Countdown(MID_Countdown:Material Instance Dynamic Object Reference=@n1) -> (Output_Get:Material Instance Dynamic Object Reference)
   > SetBrushFromMaterial(self:Border Object Reference=Get CountdownBorder() -> (CountdownBorder:Border Object Reference), Material=@n2)
   Event Tick()
   > SetScalarParameterValue(
    self=Get MID_Countdown() -> (MID_Countdown:Material Instance Dynamic Object Reference)
    ParameterName=CountdownTime
    Value=GetCountdownSeconds!(self:Take Recorder Object Reference=Get Recorder() -> (Recorder:Take Recorder Object Reference)) -> (ReturnValue:float)
   )
  ]}
 GetCountdownText {nodes [
   GetCountdownText() {is=title K2Node_FunctionEntry}
   n1 = GetCountdownSeconds(self=Get Recorder())
   > Return Node!(execute, ReturnValue:Text="To Text (Integer)"(Value=Ceil(A=@n1) {is=title K2Node_CallFunction}))
  ]}
 GetCountdownVisibility {nodes [
   GetCountdownVisibility() {is=title K2Node_FunctionEntry}
   n1 = Get Recorder()
   > n2 = Branch(Condition="Equal (Enum)"(
     A:ETakeRecorderState Enum=GetState!(self:Take Recorder Object Reference=@n1) -> (ReturnValue:ETakeRecorderState Enum=PreInitialization)
     B:ETakeRecorderState Enum=CountingDown
    ))
   > Return Node(ReturnValue:ESlateVisibility Enum=Visible)
   Return Node(execute=@n2.else, ReturnValue=Collapsed)
  ]}
]
variables [MID_Countdown {type=MaterialInstanceDynamic}]
widgets {
 info {inheritedWidgetCount=0, parentClass=/Script/TakeRecorder.TakeRecorderOverlayWidget, rootWidgetClass=&CanvasPanel, widgetCount=8}
 widgets [
  CanvasPanel_0 {widgetClassPath=&CanvasPanel, widget=CanvasPanel_0}
  CountdownBorder {
   widgetClassPath=&Border
   bIsVariable=true
   parent=CanvasPanel_0
   widget=CountdownBorder
   properties {
    background {
     bIsDynamicallyLoaded=false
     drawAs=Image
     imageType=NoImage
     mirroring=NoMirror
     resourceName=None
     resourceObject=&DefaultCountdownMaterial
     tiling=NoTile
     imageSize=<32 32>
     margin=<ltrb 0 0 0 0>
     outlineSettings {
      bUseBrushTransparency=false
      roundingType=HalfHeightRadius
      width=0
      color {colorUseRule=UseColor_Specified, specifiedColor=<rgba 0 0 0 0>}
      cornerRadii=<0 0 0 0>
     }
     tintColor {colorUseRule=UseColor_Specified, specifiedColor=<rgba 1 1 1 1>}
     uVRegion {bIsValid=false, max=<0 0>, min=<0 0>}
    }
    padding=<ltrb 0 0 0 0>
   }
   slot {class=CanvasPanelSlot, properties {layoutData {alignment=<0 0>, anchors {maximum=<1 1>, minimum=<0 0>}, offsets=<ltrb 0 0 0 0>}}}
  }
  Overlay_0 {widgetClassPath=&Overlay, parent=CountdownBorder, widget=Overlay_0, slot {class=BorderSlot, properties {padding=<ltrb 0 0 0 0>}}}
  Crosshair_H {
   widgetClassPath=&Image
   bIsVariable=true
   parent=Overlay_0
   widget=Crosshair_H
   properties {
    brush {
     bIsDynamicallyLoaded=false
     drawAs=Box
     imageType=NoImage
     mirroring=NoMirror
     resourceName=None
     resourceObject=null
     tiling=Both
     imageSize=<3 3>
     margin=<ltrb 0 0 0 0>
     outlineSettings {
      bUseBrushTransparency=false
      roundingType=HalfHeightRadius
      width=0
      color {colorUseRule=UseColor_Specified, specifiedColor=<rgba 0 0 0 0>}
      cornerRadii=<0 0 0 0>
     }
     tintColor {colorUseRule=UseColor_Specified, specifiedColor=<rgba 1 1 1 1>}
     uVRegion {bIsValid=false, max=<0 0>, min=<0 0>}
    }
    colorAndOpacity=<rgba 0 0 0 1>
   }
   slot {class=OverlaySlot, properties {horizontalAlignment=HAlign_Fill, verticalAlignment=VAlign_Center}}
  }
  Crosshair_V {
   widgetClassPath=&Image
   bIsVariable=true
   parent=Overlay_0
   widget=Crosshair_V
   properties {
    brush {
     bIsDynamicallyLoaded=false
     drawAs=Box
     imageType=NoImage
     mirroring=NoMirror
     resourceName=None
     resourceObject=null
     tiling=Both
     imageSize=<3 3>
     margin=<ltrb 0 0 0 0>
     outlineSettings {
      bUseBrushTransparency=false
      roundingType=HalfHeightRadius
      width=0
      color {colorUseRule=UseColor_Specified, specifiedColor=<rgba 0 0 0 0>}
      cornerRadii=<0 0 0 0>
     }
     tintColor {colorUseRule=UseColor_Specified, specifiedColor=<rgba 1 1 1 1>}
     uVRegion {bIsValid=false, max=<0 0>, min=<0 0>}
    }
    colorAndOpacity=<rgba 0 0 0 1>
   }
   slot {class=OverlaySlot, properties {horizontalAlignment=HAlign_Center, verticalAlignment=VAlign_Fill}}
  }
  ScaleBox_0 {
   widgetClassPath=&ScaleBox
   parent=Overlay_0
   widget=ScaleBox_0
   slot {class=OverlaySlot, properties {horizontalAlignment=HAlign_Fill, verticalAlignment=VAlign_Center, padding=<ltrb 100 100 100 100>}}
  }
  Circles {
   widgetClassPath=&Image
   bIsVariable=true
   parent=ScaleBox_0
   widget=Circles
   properties {brush {
     bIsDynamicallyLoaded=false
     drawAs=Image
     imageType=NoImage
     mirroring=NoMirror
     resourceName=None
     resourceObject=&/DefaultOverlayCircles
     tiling=NoTile
     imageSize=<1024 1024>
     margin=<ltrb 0 0 0 0>
     outlineSettings {
      bUseBrushTransparency=false
      roundingType=HalfHeightRadius
      width=0
      color {colorUseRule=UseColor_Specified, specifiedColor=<rgba 0 0 0 0>}
      cornerRadii=<0 0 0 0>
     }
     tintColor {colorUseRule=UseColor_Specified, specifiedColor=<rgba 1 1 1 1>}
     uVRegion {bIsValid=false, max=<0 0>, min=<0 0>}
    }}
   slot {class=ScaleBoxSlot}
  }
  CountdownText {
   widgetClassPath=&TextBlock
   parent=Overlay_0
   widget=CountdownText
   properties {
    justification=Center
    text=5
    font {
     bForceMonospaced=false
     bMaterialIsStencil=false
     fontMaterial=null
     fontObject=&Roboto
     letterSpacing=0
     monospacedWidth=1
     size=300
     skewAmount=0
     typefaceFontName=Bold
     outlineSettings {
      bApplyOutlineToDropShadows=true
      bMiteredCorners=false
      bSeparateFillAlpha=false
      outlineMaterial=null
      outlineSize=3
      outlineColor=<rgba 0 0 0 1>
     }
    }
    shadowOffset=<0 0>
   }
   slot {class=OverlaySlot, properties {horizontalAlignment=HAlign_Center, verticalAlignment=VAlign_Center}}
  }
 ]
}
